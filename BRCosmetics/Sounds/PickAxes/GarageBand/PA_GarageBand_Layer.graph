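# SoundCue for the GarageBand pickaxe that plays a layered generic impact sound.
# A Random node picks one of five equally weighted impact waves, then a Modulator shifts it noticeably higher with randomized pitch.
# It routes to the impacts sound class at reduced volume, with overridden short-range logarithmic attenuation with low-pass filtering, and concurrency limits that stop the quietest instance.
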SoundCue /BRCosmetics/Sounds/PickAxes/GarageBand/PA_GarageBand_Layer
root=Modulator_0
nodes [
 {id=Modulator_0, children [Random_1], properties {pitchMax=1.5, pitchMin=1.3}}
 {id=Random_1, children [WavePlayer_5,WavePlayer_6,WavePlayer_7,WavePlayer_8,WavePlayer_9], properties {weights [1,1,1,1,1]}}
 {id=WavePlayer_5, properties {soundWaveAssetPtr=&/FNBR_PA_GarageBand_Gen_Impact_04}}
 {id=WavePlayer_6, properties {soundWaveAssetPtr=&/FNBR_PA_GarageBand_Gen_Impact_01}}
 {id=WavePlayer_7, properties {soundWaveAssetPtr=&/FNBR_PA_GarageBand_Gen_Impact_02}}
 {id=WavePlayer_8, properties {soundWaveAssetPtr=&/FNBR_PA_GarageBand_Gen_Impact_03}}
 {id=WavePlayer_9, properties {soundWaveAssetPtr=&/FNBR_PA_GarageBand_Gen_Impact_05}}
]
properties {
 attenuationSettings=/Game/Sounds/Attenuation/Fort_Player_Footsteps_Att
 bOverrideAttenuation=true
 bOverrideConcurrency=true
 bPrimeOnLoad=true
 maxDistance=1100
 soundClassObject=&Fort_Impacts
 volumeMultiplier=0.375
 attenuationOverrides {bAttenuateWithLPF=true, distanceAlgorithm=Logarithmic, falloffDistance=800, attenuationShapeExtents {x=300}}
 concurrencyOverrides {maxCount=3, resolutionRule=StopQuietest}
}
