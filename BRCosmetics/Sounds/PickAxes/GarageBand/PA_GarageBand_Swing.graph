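# Sound cue for the swing of the GarageBand cosmetic pickaxe.
# A random node picks one of five equally weighted swing wave variations.
# It routes to the impacts sound class with a cosmetic pickaxe property sheet and primes on load.
# Attenuation overrides add logarithmic falloff, low-pass filtering and listener focus, while concurrency overrides cap simultaneous plays by stopping the quietest.
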SoundCue /BRCosmetics/Sounds/PickAxes/GarageBand/PA_GarageBand_Swing
root=Random_0
nodes [
 {id=Random_0, children [WavePlayer_1,WavePlayer_2,WavePlayer_3,WavePlayer_4,WavePlayer_5], properties {weights [1,1,1,1,1]}}
 {id=WavePlayer_1, properties {soundWaveAssetPtr=&/PA_GarageBand_Swing_01}}
 {id=WavePlayer_2, properties {soundWaveAssetPtr=&/PA_GarageBand_Swing_02}}
 {id=WavePlayer_3, properties {soundWaveAssetPtr=&/PA_GarageBand_Swing_03}}
 {id=WavePlayer_4, properties {soundWaveAssetPtr=&/PA_GarageBand_Swing_04}}
 {id=WavePlayer_5, properties {soundWaveAssetPtr=&/PA_GarageBand_Swing_05}}
]
properties {
 attenuationSettings=/Game/Sounds/Attenuation/Fort_Player_Footsteps_Att
 audioPropertiesSheet=/BRCosmetics/Sounds/Settings/PropertySheets/Pickaxe/APS_Cosmetic_Pickaxe_Swing
 bOverrideAttenuation=true
 bOverrideConcurrency=true
 bPrimeOnLoad=true
 maxDistance=1100
 soundClassObject=&Fort_Impacts
 volumeMultiplier=0.6
 attenuationOverrides {bAttenuateWithLPF=true, bEnableListenerFocus=true, distanceAlgorithm=Logarithmic, falloffDistance=800, attenuationShapeExtents {x=300}}
 concurrencyOverrides {maxCount=3, resolutionRule=StopQuietest}
}
